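# Material function for Apollo landscape roads that layers asphalt with painted lines over an incoming aggregate material.
# MF_ApolloAsphalt supplies the base, with a switchable edge-to-center tint; a packed mask and normal map add lines broken by crack height.
# Vertex alpha, height-lerped with world-aligned noise, times the asphalt mask drives the blend; that mask is also output as Alpha.
# Lower feature levels drop the noise, and forward/mobile paths can skip the aggregate blend.
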MaterialFunction /Clyde/Athena/Apollo/Environments/Landscape/Materials/MaterialFunctions/MF_ApolloRoad
graph {
 properties {libraryCategoriesText []}
 nodes [
  n1 = GetMaterialAttributes(None=FunctionInput() {description=Aggregate Input Material, inputName=Aggregate In, inputType=FunctionInput_MaterialAttributes}) {attributeGetTypes [
    #1
    #2
    #3
    #4
   ]}
  n2 = MaterialFunctionCall!(TilingSize) {materialFunction=&/MF_ApolloAsphalt}
  n3 = GetMaterialAttributes(None=@n2.Result) {attributeGetTypes [#1, #3]}
  n4 = MaterialFunctionCall!(Mask Width Push, Gradient Mask Value) {materialFunction=/Game/Environments/Sets/Arid/Materials/MF_RoadMaskGradient}
  n5 = Multiply(
   A=@n3.BaseColor
   B=LinearInterpolate(
    A=VectorParameter() {group=Road, parameterName=RoadTint_Edge, defaultValue=<rgba 0.5625 0.5625 0.5625 1>}.RGB
    B=VectorParameter() {group=Road, parameterName=RoadTint_Center, defaultValue=<rgba 0.151 0.151 0.151 0>}.RGB
    Alpha=@n4.Result
   )
  )
  n6 = LinearInterpolate(A=Multiply(B=@n3.Roughness) {constA=1}, B=@n3.Roughness, Alpha=@n4.Result)
  n7 = GetMaterialAttributes(None=StaticSwitch(
    True=SetMaterialAttributes!(MaterialAttributes=@n3.MaterialAttributes, BaseColor=@n5, Roughness=@n6) {attributeSetTypes [#1, #3]}
    False=@n2.Result
   )) {attributeGetTypes [
    #3
    #2
    #1
    #4
   ]}
  n8 = ScalarParameter() {defaultValue=0.5, group=Road, parameterName=Lines_R2}
  n9 = TextureObject() {texture=/Game/Environments/Sets/Arid/Textures/T_Swirly_Noise_D}
  n10 = ScalarParameter() {defaultValue=350, group=Road, parameterName=RoadTransDetailMask}
  n11 = LinearInterpolate(
   A=@n8
   B=ScalarParameter() {defaultValue=0.9575, group=Road, parameterName=Lines_R}
   Alpha=ComponentMask(None=MaterialFunctionCall(TextureObject=@n9, TextureSize=@n10) {materialFunction=&WorldAlignedTexture}.Z Texture) {b=true}
  )
  n12 = TextureSampleParameter2D!(UVs, Bias=ScalarParameter() {defaultValue=-0.5, parameterName=MIpBias}, Apply View MipBias) {
   group=Road Lines
   mipValueMode=TMVM_MipBias
   parameterName=Road Line Mask
   samplerType=SAMPLERTYPE_Masks
   texture=/Clyde/Athena/Apollo/Environments/Landscape/Textures/Roads/AsphaltRoad1_rgb
  }
  n13 = LinearInterpolate(A=@n12.G, Alpha=@n2.CrackHeight) {constB=0}
  n14 = VectorParameter() {group=Road Lines, parameterName=EdgeLine, defaultValue=<rgba 0.7344 0.6848 0.6196 0>}
  n15 = VectorParameter() {group=Road Lines, parameterName=CenterLine, defaultValue=<rgba 0.4125 0.2346 0 1>}
  n16 = TextureSampleParameter2D!(UVs, Apply View MipBias) {
   group=Road Lines
   parameterName=Road Line Normal
   samplerType=SAMPLERTYPE_Normal
   texture=/Clyde/Athena/Apollo/Environments/Landscape/Textures/Roads/AsphaltRoad1_normal
  }
  n17 = Multiply(A=@n16.RGB, B=VectorParameter() {group=Road Lines, parameterName=RoadTile N Value, defaultValue=<rgba 1.7 1.7 1 0>}.RGB)
  n18 = MaterialFunctionCall(BaseNormal=@n7.Normal, AdditionalNormal=@n17) {materialFunction=&BlendAngleCorrectedNormals}
  n19 = SetMaterialAttributes!(
   MaterialAttributes=@n7.MaterialAttributes
   Roughness=LinearInterpolate(A=@n7.Roughness, B=FeatureLevelSwitch(Default=@n11, ES2=@n8, ES3_1=@n8, SM4_REMOVED=@n11), Alpha=@n13)
   Specular=LinearInterpolate(A=@n7.Specular, Alpha=@n13) {constB=0.6}
   BaseColor=LinearInterpolate(A=@n7.BaseColor, B=LinearInterpolate(A=@n14.RGB, B=@n15.RGB, Alpha=@n12.B), Alpha=@n13)
   Normal=LinearInterpolate(A=LinearInterpolate(A=@n18.Result, B=@n17, Alpha=@n13), B=@n7.Normal, Alpha=@n2.CrackHeight)
  ) {attributeSetTypes [
    #3
    #2
    #1
    #4
   ]}
  n20 = GetMaterialAttributes(None=@n19) {attributeGetTypes [
    #1
    #2
    #3
    #4
   ]}
  n21 = VertexColor()
  n22 = ComponentMask(None=MaterialFunctionCall(TextureObject=@n9, TextureSize=Multiply(B=@n10) {constA=2}) {materialFunction=&WorldAlignedTexture}.Z Texture) {g=true}
  n23 = MaterialFunctionCall(Transition Phase=@n21.A, Height Texture=Multiply(A=LinearInterpolate(Alpha=@n22) {constB=15}, B=Multiply(A=@n22) {constB=0.5})) {materialFunction=&HeightLerp}
  n24 = ScalarParameter() {
   defaultValue=1
   desc="Multiplier for asphalt mask channel (Red in road line mask texture), use values higher than 1.0 to remove the RVT landscape contribution. "
   group=Road Lines
   parameterName=Asphalt Mask Intensity
  }
  n25 = Saturate(None=Multiply(A=@n12.R, B=@n24) {constB=0.5})
  n26 = Multiply(A=FeatureLevelSwitch(Default=@n23.Alpha, ES2=@n21.A, ES3_1=@n21.A, SM4_REMOVED=@n23.Alpha), B=@n25)
  n27 = SetMaterialAttributes!(
   MaterialAttributes
   BaseColor=LinearInterpolate(A=@n1.BaseColor, B=@n20.BaseColor, Alpha=@n26)
   Specular=LinearInterpolate(A=@n1.Specular, B=@n20.Specular, Alpha=@n26)
   Roughness=LinearInterpolate(A=@n1.Roughness, B=@n20.Roughness, Alpha=@n26)
   Normal=LinearInterpolate(A=@n1.Normal, B=@n20.Normal, Alpha=@n26)
  ) {attributeSetTypes [
    #1
    #2
    #3
    #4
   ]}
  n28 = StaticSwitchParameter(True=@n27, False=@n19) {defaultValue=true, group=Mobile, parameterName=Mobile Aggregate}
  FunctionOutput(None=ShadingPathSwitch(Default=@n27, Deferred=@n27, Forward=@n28, Mobile=@n28))
  FunctionOutput(None=@n25) {outputName=Alpha, sortPriority=1}
  StaticSwitch(True=@n27, False=@n28) {desc=debug mobile}
 ]
}
