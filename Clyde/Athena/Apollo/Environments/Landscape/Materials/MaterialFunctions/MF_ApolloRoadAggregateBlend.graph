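# Material function for Apollo landscape roads that outputs an Alpha blending aggregate into an asphalt mask.
# Aggregate weight comes from red vertex color, scaled down on forward and mobile, or from a centre-peaked U gradient.
# World-aligned noise modulates that weight before it is added to Asphalt Alpha.
# Static switches toggle aggregate and an optional road fade along V, broken up by a contrast-adjusted crack texture.
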MaterialFunction /Clyde/Athena/Apollo/Environments/Landscape/Materials/MaterialFunctions/MF_ApolloRoadAggregateBlend
graph {nodes [
  n1 = TextureCoordinate()
  n2 = ComponentMask(None=@n1) {g=true}
  n3 = StaticSwitchParameter(True=OneMinus(None=@n2), False=@n2) {defaultValue=true, group=Aggregate Blend, parameterName=Fade From Top}
  n4 = FunctionInput(Preview=Constant() {r=1}) {bUsePreviewValueAsDefault=true, inputName=Aggregate Alpha, inputType=FunctionInput_Scalar}
  n5 = Multiply(A=@n4, B=@n4) {constB=1.25}
  n6 = VertexColor()
  n7 = Multiply(A=@n6.R, B=ScalarParameter() {defaultValue=0.5, group=Mobile, parameterName=Mobile Aggregate Blend Amount}) {constB=0}
  n8 = Multiply(A=ComponentMask(None=TextureCoordinate()) {r=true}) {constB=2}
  n9 = OneMinus(None=Abs(None=Subtract(A=@n8)))
  n10 = ScalarParameter() {defaultValue=2, parameterName=Gradient offset, sliderMax=10, sliderMin=0.1}
  n11 = Subtract(B=Power(
    Base=Multiply(A=@n9, B=@n10) {constB=5}
    Exp=ScalarParameter() {defaultValue=0.8, parameterName=Gradient Exponent, sliderMax=4, sliderMin=0.5}
   )) {constB=4}
  n12 = StaticSwitchParameter(True=ShadingPathSwitch(Default=@n6.R, Deferred=@n6.R, Forward=@n7, Mobile=@n7), False=Saturate(None=OneMinus(None=@n11))) {group=Aggregate Blend, parameterName=Vert Color For Aggregate}
  n13 = Multiply(A=@n5, B=@n12)
  n14 = MaterialFunctionCall!(
   WorldPosition=WorldPosition() {worldPositionShaderOffset=WPT_CameraRelativeNoOffsets}.XYZ
   TextureSize=ScalarParameter() {defaultValue=1500, group=Aggregate, parameterName=Noise_TilingSize}
  ) {materialFunction=/Game/Environments/Sets/Arid/Materials/WorldAlignedUVs}
  n15 = MaterialFunctionCall(X=@n14.X, Y=@n14.X) {materialFunction=&MakeFloat2}
  n16 = StaticSwitchParameter(
   True=TextureSampleParameter2D(UVs=@n15.Result) {group=Aggregate Blend, parameterName=Aggregate Noise1, texture=/Engine/MaterialTemplates/Textures/T_Noise01}.R
   False=TextureSampleParameter2D(UVs=@n15.Result) {group=Aggregate Blend, parameterName=Aggregate Noise2, texture=/Engine/MapTemplates/TilingNoise05}.R
  )
  n17 = ScalarParameter() {defaultValue=1, parameterName=Aggregate Noise Strength, sliderMax=2, sliderMin=-1}
  n18 = Multiply(A=@n12, B=Power(Base=Multiply(A=@n16, B=@n17) {constB=0.5}, Exp=ScalarParameter() {defaultValue=1.25, parameterName=Exponent Height}))
  n19 = Add(A=Power(Base=@n13), B=Add(A=@n13, B=Multiply(A=@n5, B=@n18)) {constB=0.5})
  n20 = FunctionInput(Preview=Constant() {r=1}) {bUsePreviewValueAsDefault=true, inputName=Asphalt Alpha, inputType=FunctionInput_Scalar}
  n21 = StaticSwitchParameter(True=Saturate(None=Add(A=@n20, B=@n19)), False=@n20) {defaultValue=true, group=Aggregate Blend, parameterName=Aggregate ON}
  n22 = Multiply(A=Power(Base=Multiply(A=@n3) {constB=1.75}) {constExponent=1.4}, B=@n21)
  n23 = Multiply(
   A=MaterialFunctionCall!(
    Color=TextureSample(UVs=@n1) {texture=/Clyde/Packages/DS_Fortnight/Cracks}.G
    Contrast=ScalarParameter() {defaultValue=0.3, parameterName=BreakupContrast1}
   ) {materialFunction=/Engine/Functions/Engine_MaterialFunctions01/ImageAdjustment/Contrast_Preserve_Color}.Result
   B=@n22
  )
  n24 = ScalarParameter() {defaultValue=3, parameterName=BreakupContrast2}
  n25 = Multiply(
   A=Power(Base=MaterialFunctionCall(Color=@n23, Contrast=@n24) {materialFunction=&Contrast_Preserve_Color}.Result) {constExponent=1.5}
   B=@n22
  )
  FunctionOutput(None=StaticSwitchParameter(True=Saturate(None=@n25), False=@n21) {group=Aggregate Blend, parameterName=Fade Road}) {outputName=Alpha}
 ]}
