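# Material function that builds the gravel aggregate layer for Apollo road materials from roadside gravel diffuse, SMR and normal textures exposed as parameters.
# It tiles them in world space via WorldAlignedTextures_Complex, using only the Z-plane projection and world position without shader offsets.
# The results become material attributes, with specular and roughness from SMR red and blue.
# SMR green is a separate Alpha output.
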MaterialFunction /Clyde/Athena/Apollo/Environments/Landscape/Materials/MaterialFunctions/MF_ApolloRoad_Aggregate
graph {nodes [
  n1 = TextureObjectParameter() {
   group=Aggregate
   parameterName=Aggregate Diffuse
   texture=/Game/Athena/Apollo/Environments/BuildingActors/Bridges/Walls/Textures/T_RoadsideGravel_D
  }
  n2 = TextureObjectParameter() {
   group=Aggregate
   parameterName=Aggregate SMR
   samplerType=SAMPLERTYPE_Masks
   texture=/Game/Athena/Apollo/Environments/BuildingActors/Bridges/Walls/Textures/T_RoadsideGravel_S
  }
  n3 = TextureObjectParameter() {
   group=Aggregate
   parameterName=Aggregate Normal
   samplerType=SAMPLERTYPE_Normal
   texture=/Game/Athena/Apollo/Environments/BuildingActors/Bridges/Walls/Textures/T_RoadsideGravel_N
  }
  n4 = MaterialFunctionCall!(
   Diffuse Texture Object=@n1
   Spec Texture Object=@n2
   TextureSize=ScalarParameter() {defaultValue=620, group=Aggregate, parameterName=Aggregate_TilingSize}
   --------------------------
   Use seperate settings for Z projection
   Use seperate texture for Z projection
   Texture Size Z projection
   Z Plane Diffuse Texture Object
   Z Plane Spec Texture Object
   "--------------------------      "
   ProjectionTransitionContrast
   Use Texture Alpha Heightmap to Refine the Transistion=StaticBool()
   HeightLerpTransitionContrast
   PickTheDiffuseHeightChannel
   TheDiffuseHeightChannel
   "--------------------------       "
   NormalTextureObject=@n3
   World Space Normals=StaticBool()
   "--------------------------  "
   Export Float 4 Diffuse=StaticBool()
   Export Float 4 Spec
   "--------------------------     "
   WorldPosition=WorldPosition() {worldPositionShaderOffset=WPT_ExcludeAllShaderOffsets}.XYZ
   World Space Normal
  ) {materialFunction=/Engine/Functions/Engine_MaterialFunctions02/Texturing/WorldAlignedTextures_Complex}
  n5 = ComponentMask(None=@n4.Z Diffuse Texture) {b=true, g=true, r=true}
  n6 = SetMaterialAttributes!(
   MaterialAttributes
   BaseColor=ShadingPathSwitch(Default=@n5, Deferred=@n5, Forward=@n5, Mobile=@n5)
   Specular=ComponentMask(None=@n4.Z Spec) {r=true}
   Roughness=ComponentMask(None=@n4.Z Spec) {b=true}
   Normal=@n4.Normal Z Texture
  ) {attributeSetTypes [
    #1
    #2
    #3
    #4
   ]}
  FunctionOutput(None=@n6)
  FunctionOutput(None=ComponentMask(None=@n4.Z Spec) {g=true}) {outputName=Alpha, sortPriority=1}
 ]}
