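# Master road material for Apollo spline and instanced meshes, layering aggregate and road functions into an asphalt surface.
# Forward and mobile paths use a simpler blend over mobile asphalt or gravel, chosen by static switch.
# The surface writes to runtime virtual texture and can sample the landscape RVT, optionally fading along UV V.
# Optional switches add world position offset at distance or sink proxy roads near the camera using depth-based sphere masks.
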Material /Clyde/Athena/Apollo/Environments/Landscape/Materials/M_Apollo_Roads_Master
graph {
 properties {
  bCastRayTracedShadows=false
  bTangentSpaceNormal=false
  bUseMaterialAttributes=true
  bUsedWithInstancedStaticMeshes=true
  bUsedWithNanite=true
  bUsedWithSplineMeshes=true
  physMaterial=/Game/FortressPhysicalMaterials/Asphalt
  pixelDepthOffsetMode=PDOM_Legacy
  usedShadingModels=Default Lit
 }
 nodes [
  ComponentMask(None=WorldPosition().XYZ) {b=true}
  n1 = MaterialFunctionCall() {materialFunction=/Clyde/Athena/Apollo/Environments/Landscape/Materials/MaterialFunctions/MF_ApolloRoad_Aggregate}
  n2 = MaterialFunctionCall!(Aggregate In=@n1.Result) {materialFunction=/Clyde/Athena/Apollo/Environments/Landscape/Materials/MaterialFunctions/MF_ApolloRoad}
  n3 = GetMaterialAttributes(None=StaticSwitchParameter(
    True=MaterialFunctionCall!(TilingSize) {materialFunction=/Clyde/Athena/Apollo/Environments/Landscape/Materials/MaterialFunctions/MF_ApolloAsphalt}.Result
    False=MaterialFunctionCall!(UseLandscapeCoords) {materialFunction=/Clyde/Athena/Apollo/Environments/Landscape/Materials/MaterialFunctions/MF_Mobile_Common_Gravel_Roads}.Material
   ) {defaultValue=true, group=Mobile, parameterName=Mobile Blend to Asphalt}) {attributeGetTypes [
    #1
    #2
    #3
    #4
   ]}
  n4 = GetMaterialAttributes(None=@n2.Result) {attributeGetTypes [
    #1
    #2
    #3
    #4
   ]}
  n5 = MaterialFunctionCall!(Asphalt Alpha=@n2.Alpha, Aggregate Alpha=@n1.Alpha) {materialFunction=/Clyde/Athena/Apollo/Environments/Landscape/Materials/MaterialFunctions/MF_ApolloRoadAggregateBlend}
  n6 = Saturate(None=@n5.Alpha)
  n7 = SetMaterialAttributes!(
   MaterialAttributes
   Base Color=LinearInterpolate(A=@n3.Base Color, B=@n4.Base Color, Alpha=@n6)
   Specular=LinearInterpolate(A=@n3.Specular, B=@n4.Specular, Alpha=@n6)
   Roughness=LinearInterpolate(A=@n3.Roughness, B=@n4.Roughness, Alpha=@n6)
   Normal=LinearInterpolate(A=@n3.Normal, B=@n4.Normal, Alpha=@n6)
  ) {attributeSetTypes [
    #1
    #2
    #3
    #4
   ]}
  n8 = ShadingPathSwitch(Default=@n2.Result, Deferred=@n2.Result, Forward=@n7, Mobile=@n7)
  n9 = MaterialFunctionCall!(Material=@n8) {materialFunction=/Game/Athena/Apollo/Maps/Landscape/VT/MaterialFunctions/TransformTangentToWorld}
  MaterialFunctionCall(MA In=@n9.Result) {materialFunction=&MF_CreativeMoveTool}
  n10 = GetMaterialAttributes(None=@n9.Result) {attributeGetTypes [
    #1
    #2
    #3
    #4
   ]}
  RuntimeVirtualTextureOutput!(
   BaseColor=@n10.Base Color
   Specular=@n10.Base Color
   Roughness=@n10.Base Color
   Normal=@n10.Base Color
   WorldHeight
   Opacity=@n5.Alpha
   Mask
   Displacement
   Mask4
  )
  n11 = Constant3Vector() {constant=<rgba 0 0 4 0>}
  n12 = Distance(A=CameraPositionWS(), B=WorldPosition().XYZ)
  n13 = ScalarParameter() {defaultValue=2500, group=Mobile, parameterName=Radius of Mobile Height Bias}
  n14 = Min(A=Floor(None=Divide(A=@n12, B=@n13))) {constB=10}
  n15 = Multiply(A=ScalarParameter() {defaultValue=2.75, group=Mobile, parameterName=Mobile Aggregate Lower Amount}, B=OneMinus(None=VertexColor().R))
  SetMaterialAttributes!(MaterialAttributes=@n9.Result, World Position Offset=Subtract(A=Multiply(A=@n11, B=@n14), B=Multiply(A=@n11, B=@n15))) {attributeSetTypes [#5]}
  n16 = RuntimeVirtualTextureSample!(UVs, World Position) {materialType=BaseColor_Normal_Specular_YCoCg, virtualTexture=/Game/Athena/Apollo/RVT/RVT_Apollo_Landscape}
  n17 = SetMaterialAttributes!(MaterialAttributes, Base Color=@n16.BaseColor, Specular=@n16.BaseColor, Roughness=@n16.BaseColor, Normal=@n16.BaseColor) {attributeSetTypes [
    #1
    #2
    #3
    #4
   ]}
  n18 = GetMaterialAttributes(None=@n17) {attributeGetTypes [
    #1
    #2
    #3
    #4
   ]}
  n19 = GetMaterialAttributes(None=@n10.MaterialAttributes) {attributeGetTypes [
    #1
    #2
    #3
    #4
   ]}
  n20 = ComponentMask(None=TextureCoordinate()) {g=true}
  n21 = OneMinus(None=@n20)
  n22 = Multiply(A=@n5.Alpha, B=Saturate(None=StaticSwitchParameter(True=@n20, False=@n21) {group=Non RVT to RVT Fade, parameterName=V direction Fade UP}))
  n23 = SetMaterialAttributes(
   Base Color=LinearInterpolate(A=@n18.Base Color, B=@n19.Base Color, Alpha=@n22)
   Specular=LinearInterpolate(A=@n18.Specular, B=@n19.Specular, Alpha=@n22)
   Roughness=LinearInterpolate(A=@n18.Roughness, B=@n19.Roughness, Alpha=@n22)
   Normal=LinearInterpolate(A=@n18.Normal, B=@n19.Normal, Alpha=@n22)
  ) {attributeSetTypes [
    #1
    #2
    #3
    #4
   ]}
  n24 = StaticSwitchParameter(
   True=MaterialFunctionCall!(RVT=StaticSwitchParameter(True=@n23, False=@n17) {group=Non RVT to RVT Fade, parameterName=Non RVT Fade to RVT}, ByPass RVT=@n9.Result) {materialFunction=/Clyde/Athena/Apollo/Maps/Landscape/VT/MaterialFunctions/EnableRVT}.Result
   False=@n9.Result
  ) {defaultValue=true, group=RVT Base, parameterName=RVT Enable}
  n25 = MaterialFunctionCall(Z=SphereMask(A=PixelDepth().R, B=Constant()) {attenuationRadius=22000}) {materialFunction=&MakeFloat3}
  n26 = LinearInterpolate(
   A=Constant3Vector()
   B=VectorParameter() {parameterName=WPO Input, defaultValue=<rgba 0 0 50 0>}.RGB
   Alpha=OneMinus(None=Clamp(None=@n25.Result))
  )
  n27 = MaterialFunctionCall(
   NoRVT=@n9.Result
   RVT=StaticSwitchParameter(
    True=MaterialFunctionCall!(Attributes=@n24, Additional WPO=@n26) {materialFunction=/FortWater/Materials/Functions/AttributesAddWPO}.Result
    False=@n24
   ) {parameterName=Use WPO at Distance}
  ) {materialFunction=&FN_RVT_Switch_Material}
  n28 = BreakMaterialAttributes!(Attr=@n27.Result)
  n29 = Constant3Vector()
  n30 = VectorParameter() {parameterName=ProxyOffset, defaultValue=<rgba 0 0 -500 1>}
  n31 = PixelDepth()
  n32 = Constant()
  n33 = ScalarParameter() {defaultValue=5000, group=ProxyRoad, parameterName=MaxProxyOffsetRange}
  n34 = Clamp(None=MaterialFunctionCall(Z=SphereMask(A=@n31.R, B=@n32, Radius=@n33) {attenuationRadius=22000, hardnessPercent=20}) {materialFunction=&MakeFloat3}.Result)
  StaticSwitchParameter(
   True=MakeMaterialAttributes!(
    BaseColor=@n28.BaseColor
    Metallic=@n28.BaseColor
    Specular=@n28.BaseColor
    Roughness=@n28.BaseColor
    Anisotropy=@n28.BaseColor
    EmissiveColor=@n28.BaseColor
    Opacity=@n28.BaseColor
    OpacityMask=@n28.BaseColor
    Normal=@n28.BaseColor
    Tangent=@n28.BaseColor
    WorldPositionOffset=LinearInterpolate(A=@n29, B=@n30.RGB, Alpha=@n34)
    SubsurfaceColor=@n28.BaseColor
    ClearCoat=@n28.BaseColor
    ClearCoatRoughness=@n28.BaseColor
    AmbientOcclusion=@n28.BaseColor
    Refraction=@n28.BaseColor
    CustomizedUVs_0
    CustomizedUVs_1
    CustomizedUVs_2
    CustomizedUVs_3
    CustomizedUVs_4
    CustomizedUVs_5
    CustomizedUVs_6
    CustomizedUVs_7
    PixelDepthOffset=@n28.BaseColor
    ShadingModel=@n28.BaseColor
    Displacement=@n28.BaseColor
   )
   False=@n27.Result
  ) {parameterName=UseProxyFade}
  StaticSwitch(True=@n8, False=@n7) {defaultValue=true, desc=Debug Mobile}
 ]
}
