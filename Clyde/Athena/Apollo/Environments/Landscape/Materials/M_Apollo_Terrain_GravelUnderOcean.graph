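# Apollo terrain material for gravel under the ocean, built as a Default Lit material from shared sand biome functions.
# MF_Common_Sand, with its under-ocean sand option switched on, supplies the Default and Deferred shading paths.
# MF_Mobile_Common_Sand supplies the cheaper Forward and Mobile paths.
# A shading path switch selects between the two full material-attribute outputs, so the graph itself holds no surface logic.
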Material /Clyde/Athena/Apollo/Environments/Landscape/Materials/M_Apollo_Terrain_GravelUnderOcean
graph {
 properties {bUseMaterialAttributes=true, pixelDepthOffsetMode=PDOM_Legacy, usedShadingModels=Default Lit}
 nodes [
  n1 = MaterialFunctionCall!(UseBiomeColorMask, BiomeColorMask, BiomeColorDiv, BiomeColor, UseUnderOceanSandMaterial=StaticBool() {value=true}) {materialFunction=/Clyde/Athena/Apollo/Environments/Landscape/Materials/Biome/CommonMaterials/MF_Common_Sand}
  n2 = MaterialFunctionCall!(CustomSeaLevel, BypassBiomeColors, UseOceanLevel) {materialFunction=/Game/Athena/Apollo/Environments/Landscape/Materials/Biome/CommonMaterials/Mobile/MF_Mobile_Common_Sand}
  ShadingPathSwitch(Default=@n1.Material, Deferred=@n1.Material, Forward=@n2.Material, Mobile=@n2.Material)
 ]
}
